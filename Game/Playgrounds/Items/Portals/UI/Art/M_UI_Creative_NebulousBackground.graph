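# Translucent UI material that draws an animated purple nebulous background for Creative portal screens.
# A screen-space eased color gradient underlies two slowly counter-rotating particle grids of randomized, morphing triangle SDFs with soft edges.
# Aurora lines come from a panning caustics texture, stretched and masked by a rotated box SDF.
# Eased vertical gradient masks are also built, and the composite is multiplied by vertex color, with parameters for colors, opacities, seeds and speed.
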Material /Game/Playgrounds/Items/Portals/UI/Art/M_UI_Creative_NebulousBackground
graph {
 properties {blendMode=BLEND_Translucent, materialDomain=MD_UI, pixelDepthOffsetMode=PDOM_Legacy, usedShadingModels=Default Lit}
 nodes [
  n1 = MaterialFunctionCall!(ScreenUV, TextureSize, ScalePivotPoint) {materialFunction=/FortUILibrary/MaterialFunctions/Basics/MF_UI_ScreenSpaceUVScaling}
  n2 = ComponentMask(None=MaterialFunctionCall(UVs=@n1.Scaled to Fit, Texture Scale=ScalarParameter() {defaultValue=1.5, parameterName=Radial Background Scale}) {materialFunction=&ScaleUVsByCenter}.UVs) {g=true}
  n3 = VectorParameter() {parameterName=Primary Background Color, defaultValue=<rgba 0.0015 0.0001 0.0052 1>}
  n4 = LinearInterpolate(
   A=@n3.RGB
   B=VectorParameter() {parameterName=Secondary Background Color, defaultValue=<rgba 0.0545 0 0.2831 1>}.RGB
   Alpha=NamedRerouteDeclaration(None=MaterialFunctionCall("Value (0-1)"=@n2) {materialFunction=&MF_UI_EaseCurves}.EaseInOutSine) {name=Background Gradient Mask, nodeColor=<rgba 0 0 0 1>}
  )
  n5 = VectorParameter() {parameterName=Triangle Background Color, defaultValue=<rgba 0.1904 0 0.9896 1>}
  n6 = NamedRerouteUsage()
  n7 = LinearInterpolate(
   A=@n3.RGB
   B=LinearInterpolate(
    A=LinearInterpolate(A=@n4, B=@n5.RGB, Alpha=@n6)
    B=VectorParameter() {parameterName=Triangle Midground Color, defaultValue=<rgba 0.0938 0 0.4824 1>}.RGB
    Alpha=NamedRerouteUsage()
   )
   Alpha=NamedRerouteUsage()
  )
  n8 = VectorParameter() {parameterName=Aurora Lines Color, defaultValue=<rgba 0.1122 0 0.5831 1>}
  n9 = NamedRerouteUsage()
  n10 = NamedRerouteUsage()
  n11 = NamedRerouteUsage()
  n12 = LinearInterpolate(A=NamedRerouteUsage(), Alpha=LinearInterpolate(A=NamedRerouteUsage(), Alpha=LinearInterpolate(A=@n10, Alpha=@n11) {constAlpha=0}))
  n13 = NamedRerouteUsage()
  Multiply(A=Multiply(A=LinearInterpolate(A=@n7, B=@n8.RGB, Alpha=@n9), B=LinearInterpolate(B=@n12, Alpha=@n13)) {constB=0.5}, B=VertexColor())
  n14 = MaterialFunctionCall(
   Input=ComponentMask(None=MaterialFunctionCall() {materialFunction=&GetUserInterfaceUV}.Normalized UV) {g=true}
   Input Low=Constant()
   Input High=ScalarParameter() {defaultValue=1, parameterName=Top Gradient Control}
  ) {materialFunction=&RemapValueRange}
  NamedRerouteDeclaration(None=MaterialFunctionCall("Value (0-1)"=Saturate(None=@n14.Result)) {materialFunction=&MF_UI_EaseCurves}.EaseInSine) {name=Top Gradient Mask, nodeColor=<rgba 0 0 0 1>}
  n15 = ScalarParameter() {parameterName=Triangle X Offset}
  n16 = ScalarParameter() {parameterName=Triangle Y Offset}
  n17 = MaterialFunctionCall(ScreenUV=Subtract(A=ScreenPosition().ViewportUV, B=AppendVector(A=@n15, B=@n16)) {constB=0}) {materialFunction=&MF_UI_ScreenSpaceUVScaling}
  n18 = NamedRerouteUsage()
  n19 = MaterialFunctionCall!(
   UV=Multiply(A=@n17.Scaled to Fit) {constB=0.9}
   Global Offset
   Global Rotation=Multiply(A=Time()) {constB=0.005}
   Tiling=Constant() {r=2}
   Cell Offset=Constant2Vector()
   Cell Scale=LinearInterpolate(Alpha=@n18) {constA=-0.1, constB=0.1}
   Cell Rotation=LinearInterpolate(Alpha=@n18) {constA=-1}
   Random Seed=ScalarParameter() {defaultValue=10, parameterName=Midground Random Seed}
  ) {materialFunction=/FortUILibrary/MaterialFunctions/Basics/MF_UI_ParticleGrid}
  NamedRerouteDeclaration(None=@n19.Random Cell Values) {name=Random Cell Value, nodeColor=<rgba 0 0 0 1>}
  n20 = NamedRerouteUsage()
  n21 = MaterialFunctionCall(
   UV=@n17.Scaled to Fit
   Global Rotation=Multiply(A=Time()) {constB=-0.005}
   Tiling=Constant() {r=3}
   Cell Offset=LinearInterpolate(Alpha=@n20) {constA=0.5}
   Cell Scale=LinearInterpolate(Alpha=@n20) {constA=-0.25, constB=0.1}
   Cell Rotation=LinearInterpolate(Alpha=@n20) {constA=-0.25, constB=0.25}
   Random Seed=ScalarParameter() {defaultValue=56, parameterName=Background Random Seed}
  ) {materialFunction=&MF_UI_ParticleGrid}
  NamedRerouteDeclaration(None=@n21.Random Cell Values) {name=Random Cell Value 3, nodeColor=<rgba 0 0 0 1>}
  n22 = Time()
  n23 = Cosine(None=Multiply(A=@n22) {constB=-0.1})
  n24 = Sine(None=Multiply(A=Add(A=@n22) {constB=0.2}) {constB=0.1})
  n25 = NamedRerouteUsage()
  n26 = Cosine(None=@n25)
  n27 = LinearInterpolate(Alpha=@n25) {constA=-1}
  n28 = Sine(None=@n25)
  n29 = SmoothStep(
   Max=LinearInterpolate(Alpha=NamedRerouteUsage()) {constA=0.01, constB=0.075}
   Value=MaterialFunctionCall!(
    UVs=@n21.Cells UV
    Corner 01=AppendVector(A=@n23, B=@n24)
    Corner 02=AppendVector(A=@n26, B=@n27)
    Corner 03=AppendVector(A=@n27, B=@n28)
    Glow Min
    Glow Max
    Outline Thickness
   ) {materialFunction=/FortUILibrary/MaterialFunctions/Shapes/MF_UI_WIP_Triangle}.SDF
  ) {constMax=0.01}
  NamedRerouteDeclaration(None=Multiply(A=OneMinus(None=@n29), B=ScalarParameter() {defaultValue=1, parameterName=Triangle Opacity, sliderMax=1})) {name=Triangle Background, nodeColor=<rgba 0 0 0 1>}
  n30 = LinearInterpolate(Alpha=NamedRerouteUsage()) {constA=-1}
  n31 = NamedRerouteUsage()
  n32 = Multiply(A=Cosine(None=@n31), B=@n31)
  n33 = NamedRerouteUsage()
  n34 = Multiply(A=Sine(None=@n33), B=@n33)
  n35 = Time()
  n36 = Add(A=@n35) {constB=0.25}
  n37 = AppendVector(A=Sine(None=Multiply(A=@n35) {constB=0.1}), B=Cosine(None=Multiply(A=@n36) {constB=-0.05}))
  n38 = NamedRerouteUsage()
  n39 = SmoothStep(
   Max=LinearInterpolate(Alpha=NamedRerouteUsage()) {constA=0.01, constB=0.05}
   Value=MaterialFunctionCall(UVs=@n19.Cells UV, Corner 01=Multiply(A=@n37, B=@n38), Corner 02=AppendVector(A=@n34, B=@n30), Corner 03=AppendVector(A=@n30, B=@n32)) {materialFunction=&MF_UI_WIP_Triangle}.SDF
  ) {constMax=0.01}
  NamedRerouteDeclaration(None=Multiply(A=OneMinus(None=@n39), B=ScalarParameter() {defaultValue=1, parameterName=Triangle Opacity, sliderMax=1})) {name=Triangle Midground, nodeColor=<rgba 0 0 0 1>}
  n40 = Subtract(A=ScreenPosition().ViewportUV, B=AppendVector(A=Constant() {r=0.1}, B=Constant() {r=0.6})) {constB=0.2}
  n41 = Rotator(
   Coordinate=MaterialFunctionCall(ScreenUV=Multiply(A=@n40) {constB=0.9}) {materialFunction=&MF_UI_ScreenSpaceUVScaling}.Scaled to Fill
   Time=Constant() {r=0.05}
  ) {speed=6.2832}
  n42 = Multiply(A=MaterialFunctionCall(Float2=@n41) {materialFunction=&BreakOutFloat2Components}.R) {constB=0.5}
  n43 = LinearInterpolate(A=Constant2Vector() {g=1, r=4}, B=Constant2Vector() {g=0.5, r=4}, Alpha=Saturate(None=Sine(None=@n42))) {constA=1}
  n44 = MaterialFunctionCall(UVs=@n41, Size=@n43) {materialFunction=&MF_UI_Box}
  n45 = Panner(Coordinate=@n41, Speed=AppendVector(A=ScalarParameter() {defaultValue=-0.15, parameterName=Aurora Lines Speed}, B=Constant())) {speedX=-1}
  n46 = Multiply(
   A=OneMinus(None=SmoothStep(Value=@n44.SDF) {constMax=0.25})
   B=TextureSample(
    UVs=MaterialFunctionCall(UVs=@n45, Texture Scale=@n43) {materialFunction=&ScaleUVsByCenter}.UVs
    Tex=NamedRerouteUsage()
   ) {samplerSource=SSM_Wrap_WorldGroupSettings, texture=&T_ShieldCaustics}.R
  )
  n47 = Saturate(None=Multiply(A=@n46, B=@n46))
  NamedRerouteDeclaration(None=Multiply(A=Multiply(A=@n47) {constB=0.75}, B=ScalarParameter() {defaultValue=1, parameterName=Wind Opacity, sliderMax=1})) {name=BG Lines, nodeColor=<rgba 0 0 0 1>}
  n48 = MaterialFunctionCall(
   Input=ComponentMask(None=MaterialFunctionCall() {materialFunction=&GetUserInterfaceUV}.Normalized UV) {g=true}
   Input Low=Constant() {r=0.8}
   Input High=Constant() {r=1}
  ) {materialFunction=&RemapValueRange}
  NamedRerouteDeclaration(None=MaterialFunctionCall("Value (0-1)"=Saturate(None=@n48.Result)) {materialFunction=&MF_UI_EaseCurves}.EaseInOutSine) {name=Overlay Gradient Mask, nodeColor=<rgba 0 0 0 1>}
  NamedRerouteDeclaration(None=@n1.ScaleX) {name=Screenspace Scale X, nodeColor=<rgba 0 0 0 1>}
 ]
}
